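# Niagara system for the pizza in the Bass "Let Em Cook" emote, with one mesh-particle emitter that bursts toppings across a flat disc.
# Spawn index picks between two meshes, and each topping tumbles at a random rate.
# User.ToppingsPositionLerpAlpha blends a random height offset onto the spawn position, so an external driver controls the toss.
# User.NormalizedNotifyProgress indexes a scale curve that shrinks the toppings to zero near the end.
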NiagaraSystem /SparksCosmetics/Emotes/Bass/LetEmCook/Effects/NS_Emote_Bass_LetEmCook_Pizza
dependencies {
 usedDataInterfaces [&NiagaraDataInterfaceCurve, &NiagaraDataInterfaceSkeletalMesh]
 usedDynamicInputs [&ReturnNormalizedExecIndex]
 usedModules [
  &self:SystemSpawnScript.SystemScriptSource.SystemScriptGraph.NiagaraNodeAssignment_2.SetVariables_#1
  &SystemState
 ]
}
emitters {Toppings {
  emitterUpdateScript [
   &EmitterState()
   &SpawnBurst_Instantaneous(Spawn Count=15, Use Spawn Probability=false, UseLoopCountLimit=true)
  ]
  particleSpawnScript [
   &InitializeParticle(Color Mode=::NewEnumerator0, Mesh Scale Mode=::NewEnumerator3, UsePositionOffset=false, Write Material Random=true)
   &ShapeLocation(
    Cylinder Height=0
    Cylinder Radius=14
    MaskSpawnGroup=false
    Override Seed=false
    Shape Primitive=::NewEnumerator1
    Surface Only=false
    UseOwnerScale=true
   )
   /SetVariables(Particles.MeshIndex=&MakeCustomIntFromBool(
     Bool=&SetBoolByFloatComparison(A=&MakeFloatFromInt(Int=&ReturnExecIndex()), B=5)
     True INT=1
     False INT=0
    )) {isSetParameters=true}
  ]
  particleUpdateScript [
   &ParticleState()
   &UpdateMeshOrientation(
    Rotation Rate=&RandomRangeFloat(Minimum=2, Maximum=3)
    Rotation Vector=&RandomRangeVector(Minimum=<-1 -1 0>, Maximum=<1 1 0>)
   )
   /SetVariables(Particles.Position=&AddVectorToPosition(
     Position=&AddVectorToPosition(Position=$Particles.Initial.Position, Vector=<0 0 1.5>)
     Vector=&TransformVector(
      Vector=&Lerp_Vector(
       A=<0 0 0>
       B=&RandomRangeVector(Minimum=<0 0 0>, Maximum=<0 0 60>)
       Alpha=$User.ToppingsPositionLerpAlpha
      )
      Source Space=::World
      Destination Space=::Local
     )
    )) {isSetParameters=true}
   &Drag(Drag=0, Use Linear Drag=true, Use Rotational Drag=true)
   &ScaleMeshSize(Scale Factor=&VectorFromFloat(Value=&FloatFromCurve(
      FloatCurve=DataInterface() {propertyValues {Curve {keys [<at 0 1>, {arriveTangent=-1, interpMode=RCIM_Cubic, leaveTangent=-1, time=0.9, value=1}, <at 1 0 RCIM_Cubic>]}}}
      CurveIndex=$User.NormalizedNotifyProgress
     )))
   &SolveForcesAndVelocity(Clamp Velocity=false, Limit Acceleration=false)
  ]
  rendererClasses [&NiagaraMeshRendererProperties]
  renderers [{rendererClass=&NiagaraMeshRendererProperties, rendererIndex=0}]
 }}
summary {
 systemName=NS_Emote_Bass_LetEmCook_Pizza
 emitters [{emitterName=Toppings, rendererClasses [&NiagaraMeshRendererProperties]}]
 userVariables [
  User.bIsFrontEndPreview {type=NiagaraBool}
  User.NormalizedNotifyProgress {type=NiagaraFloat, defaultValue=0, description=""}
  User.ToppingsPositionLerpAlpha {type=NiagaraFloat, defaultValue=0, description=""}
 ]
}
