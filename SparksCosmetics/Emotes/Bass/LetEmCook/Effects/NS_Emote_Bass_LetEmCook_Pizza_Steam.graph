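# Niagara steam effect for the pizza in the Bass "LetEmCook" emote, spawned by distance travelled.
# Steam_Ribbon leaves a widening, curl-noise trail whose spawn probability follows User.SteamRibbonSpawnScale, which defaults to zero.
# Steam_Sprites adds growing, rotating SubUV puffs from a small sphere, independent of that parameter.
# Both rise through upward gravity and drag, with fading alpha scaled by time of day to keep opacity consistent with lighting.
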NiagaraSystem /SparksCosmetics/Emotes/Bass/LetEmCook/Effects/NS_Emote_Bass_LetEmCook_Pizza_Steam
dependencies {
 usedDataInterfaces [&NiagaraDataInterfaceCurve, &NiagaraDataInterfaceSpriteRendererInfo]
 usedDynamicInputs []
 usedModules [&SystemState]
}
emitters {
 Steam_Ribbon {
  emitterUpdateScript [
   &EmitterState()
   &SpawnPerUnit(
    Module.Use Max Movement Threshold=true
    Module.Use Movement Tolerance=false
    Spawn Probability:NiagaraFloat=$User.SteamRibbonSpawnScale
    Spawn Spacing=5
    Use Spawn Probability=true
   )
  ]
  particleSpawnScript [
   &InitializeParticle(Lifetime=0.6, Ribbon Width=25, Ribbon Width Mode=::NewEnumerator1, UsePositionOffset=false, Write Material Random=true)
   &CameraOffset(Camera Offset Amount=-1)
  ]
  particleUpdateScript [
   &ParticleState()
   &GravityForce(Gravity=<0 0 500>)
   &Drag(Drag=2, Use Linear Drag=true, Use Rotational Drag=true)
   &CurlNoiseForce(Bias Noise=false, Noise Strength=50, Pan Noise=false)
   &SolveForcesAndVelocity(Clamp Velocity=false, Limit Acceleration=false)
   &ScaleColor(
    Scale Alpha=&Multiply_Float(
     A=&FloatFromCurve(FloatCurve=DataInterface() {propertyValues {Curve {keys [<at 0 1 RCIM_Cubic>,<at 0.9 0 RCIM_Cubic>]}}}, Scale Curve=0.1)
     B=&FNTimeOfDayScalar(Night Scale=0.4, Morning Scale=0.35, Day Scale=0.45, Evening Scale=0.35)
    )
    ScaleA=true
    ScaleRGB=true
   )
   &DynamicMaterialParameters(
    Index 0 Param 1=&FloatFromCurve(FloatCurve=DataInterface() {propertyValues {Curve {keys [{interpMode=RCIM_Cubic, tangentMode=RCTM_User, time=0, value=1}, {arriveTangent=0.0025, tangentMode=RCTM_User, time=1, value=0.5}]}}})
    Index 0 Param 3=-0.6
    Index 0 Param 4=1
    Module.Index0Param0Write=true
    Module.Index0Param1Write=true
    Module.Index0Param2Write=true
    Module.Index0Param3Write=true
   )
   &ScaleRibbonWidth(Ribbon Width Scale=&FloatFromCurve(FloatCurve=DataInterface() {propertyValues {Curve {keys [
         {arriveTangent=2.8633, interpMode=RCIM_Cubic, leaveTangent=2.8633, tangentMode=RCTM_User, time=0, value=1}
         {tangentMode=RCTM_User, time=1, value=2}
        ]}}}))
  ]
  rendererClasses [&NiagaraRibbonRendererProperties]
  renderers [{rendererClass=&NiagaraRibbonRendererProperties}]
 }
 Steam_Sprites {
  emitterUpdateScript [
   &EmitterState()
   &SpawnPerUnit(Module.Use Max Movement Threshold=true, Module.Use Movement Tolerance=false, Spawn Spacing=16, Use Spawn Probability=false)
  ]
  particleSpawnScript [
   &InitializeParticle(
    Lifetime Max=0.8
    Lifetime Min=0.5
    Lifetime Mode=::NewEnumerator1
    Sprite Rotation Mode=::NewEnumerator2
    Sprite Size Mode=::NewEnumerator1
    Uniform Sprite Size Max=24
    Uniform Sprite Size Min=20
    UsePositionOffset=false
    Write Material Random=true
   )
   &ShapeLocation(MaskSpawnGroup=false, Override Seed=false, Sphere Radius=5, UseOwnerScale=true)
   &CameraOffset(Camera Offset Amount=1)
   &DynamicMaterialParameters(
    Index 0 Param 1=-0.6
    Module.Index0Param0Write=true
    Module.Index0Param1Write=true
    Module.Index0Param2Write=true
    Module.Index0Param3Write=true
   )
  ]
  particleUpdateScript [
   &ParticleState()
   &ScaleSpriteSize(Uniform Curve Sprite Scale=DataInterface() {propertyValues {Curve {keys [
        {arriveTangent=6.542, interpMode=RCIM_Cubic, leaveTangent=6.542, tangentMode=RCTM_User, time=0, value=1}
        {arriveTangent=0.1082, interpMode=RCIM_Cubic, leaveTangent=0.1082, tangentMode=RCTM_User, time=1, value=3}
       ]}}})
   &GravityForce(Gravity=<0 0 300>)
   &Drag(Use Linear Drag=true, Use Rotational Drag=true)
   &SpriteRotationRate(Rotation Rate=&RandomRangeFloat(Minimum=8, Maximum=15))
   &SolveForcesAndVelocity(Clamp Velocity=false, Limit Acceleration=false)
   &SubUVAnimation(
    Random Start Frame=true
    "Sprite Renderer "=DataInterface() {propertyValues {SpriteRenderer=&self:Steam_Sprites_1.NiagaraSpriteRendererProperties_0}}
    SubUV Animation Mode:ENiagara_SubUVLookupModeV2=::NewEnumerator3
    UseEndFrame=false
    UsePlayRateScale=false
    UseStartFrame=false
   )
   &ScaleColor(
    Scale Alpha=&Multiply_Float(
     A=&FloatFromCurve(
      FloatCurve=DataInterface() {propertyValues {Curve {keys [
          {interpMode=RCIM_Cubic, tangentMode=RCTM_User, time=0, value=0}
          {arriveTangent=-3.1984, interpMode=RCIM_Cubic, leaveTangent=-3.1984, tangentMode=RCTM_User, time=0.05, value=1}
          {arriveTangent=-0.0093, tangentMode=RCTM_User, time=1, value=0}
         ]}}}
      Scale Curve=0.25
     )
     B=&FNTimeOfDayScalar(Night Scale=0.4, Morning Scale=0.35, Day Scale=0.5, Evening Scale=0.4)
    )
    ScaleA=true
    ScaleRGB=true
   )
  ]
  rendererClasses [&NiagaraSpriteRendererProperties]
  renderers [{rendererClass=&NiagaraSpriteRendererProperties}]
 }
}
summary {
 systemName=NS_Emote_Bass_LetEmCook_Pizza_Steam
 emitters [
  {emitterName=Steam_Ribbon, rendererClasses [&NiagaraRibbonRendererProperties]}
  {emitterName=Steam_Sprites, rendererClasses [&NiagaraSpriteRendererProperties]}
 ]
 userVariables [
  User.NormalizedNotifyProgress {type=NiagaraFloat, defaultValue=0.5, description=""}
  User.SteamRibbonBool {type=NiagaraBool}
  User.SteamRibbonSpawnScale {type=NiagaraFloat, defaultValue=0, description=""}
 ]
}
