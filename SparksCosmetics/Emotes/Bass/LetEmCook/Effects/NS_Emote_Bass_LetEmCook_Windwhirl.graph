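# Niagara system for the wind whirl effect of the Bass "Let Em Cook" emote, built from one mesh emitter.
# After the system delay, it bursts ten short-lived, pale blue meshes, spaced evenly along a vertical column with randomized scale and slight orientation variation.
# Each spins around Z in a random direction while curves grow its size and fade alpha in and out, scaled by time of day.
# Curves and a random value also drive dynamic material parameters.
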NiagaraSystem /SparksCosmetics/Emotes/Bass/LetEmCook/Effects/NS_Emote_Bass_LetEmCook_Windwhirl
dependencies {
 usedDataInterfaces [&NiagaraDataInterfaceVectorCurve, &NiagaraDataInterfaceCurve]
 usedDynamicInputs []
 usedModules [&SystemState]
}
emitters {Windwhirl {
  emitterUpdateScript [
   &EmitterState(
    Life Cycle Mode=::NewEnumerator1
    Loop Behavior=::NewEnumerator1
    Loop Delay=$System.Delay
    Loop Duration Mode=::NewEnumerator1
    UseLoopDelay=true
   )
   &SpawnBurst_Instantaneous(Spawn Count=10, Use Spawn Probability=false, UseLoopCountLimit=false)
  ]
  particleSpawnScript [
   &InitializeParticle(
    Color Maximum=<rgba 0.8385 0.8789 1 1>
    Color Minimum=<rgba 0.6445 0.8114 1 1>
    Color Mode=::NewEnumerator2
    Lifetime=&UniformRangedFloat(Minimum=0.1, Maximum=0.25)
    Mesh Scale=&MakeVector(
     X=&RandomRangeFloat(Minimum=0.4, Maximum=0.55)
     Y=&RandomRangeFloat(Minimum=0.4, Maximum=0.55)
     Z=&RandomRangeFloat(Minimum=0.8, Maximum=1.3)
    )
    Mesh Scale Mode=::NewEnumerator4
    Position Offset=<0 0 15>
    Position Offset Coordinate Space=::Local
    UsePositionOffset=true
    Write Material Random=true
   ) {module=InitializeParticle001}
   &InitialMeshOrientation(
    Mesh Orientation Mode=::NewEnumerator2
    Rotation=&MakeVector(X=0, Y=0, Z=&RandomRangeFloat(Minimum=-0.3, Maximum=0.3))
    Use Rotation Vector=true
   )
   /SetVariables(Particles.Position=&Lerp_Vector(A=<0 0 -65>, B=<0 0 -10>, Alpha=&ReturnNormalizedExecIndex(Normalized Index Includes One=false))) {isSetParameters=true}
   /SetVariables(Particles.FlipSide:NiagaraBool=&RandomBool()) {isSetParameters=true}
  ]
  particleUpdateScript [
   &ParticleState()
   &UpdateMeshOrientation(
    Rotation Rate=&Multiply_Float(
     A=&MakeCustomFloatFromBool(Bool=$Particles.FlipSide, True Float=-1, False Float=1)
     B=&RandomRangeFloat(Minimum=3, Maximum=4)
    )
    Rotation Vector=<0 0 1>
   )
   &ScaleMeshSize(Scale Factor=&VectorFromCurve(VectorCurve=DataInterface() {propertyValues {
       XCurve {keys [
         {arriveTangent=1.8944, interpMode=RCIM_Cubic, leaveTangent=1.8944, tangentMode=RCTM_User, time=0, value=0.2}
         {arriveTangent=0.0304, interpMode=RCIM_Cubic, leaveTangent=0.0304, tangentMode=RCTM_User, time=1, value=1}
        ]}
       YCurve {keys [
         {arriveTangent=1.9027, interpMode=RCIM_Cubic, leaveTangent=1.9027, tangentMode=RCTM_User, time=0, value=0.2}
         {arriveTangent=0.0114, interpMode=RCIM_Cubic, leaveTangent=0.0114, tangentMode=RCTM_User, time=1, value=1}
        ]}
       ZCurve {keys [
         {arriveTangent=1.0009, interpMode=RCIM_Cubic, time=0, value=0.1}
         <at 0.65 1>
         {interpMode=RCIM_Cubic, leaveTangent=0.8008, time=1, value=0.5}
        ]}
      }}))
   &ScaleColor(
    Scale Alpha=&FloatFromCurve(
     FloatCurve=DataInterface() {propertyValues {Curve {keys [
         {arriveTangent=8.125, interpMode=RCIM_Cubic, time=0, value=0}
         {
          arriveTangent=0.0113
          arriveTangentWeight=0.4805
          interpMode=RCIM_Cubic
          leaveTangent=0.0113
          leaveTangentWeight=0.1363
          tangentMode=RCTM_User
          time=0.5
          value=1
         }
         {arriveTangent=-5.7518, interpMode=RCIM_Cubic, leaveTangent=-5.7518, tangentMode=RCTM_User, time=1, value=0}
        ]}}}
     Scale Curve=&FNTimeOfDayScalar(Night Scale=0.4, Morning Scale=0.1, Day Scale=0.15, Evening Scale=0.3)
    )
    Scale RGB=&MakeVectorFromLinearColorRGB(Color=<rgba 1 1 1 1>)
    ScaleA=true
    ScaleRGB=false
   )
   &DynamicMaterialParameters(
    Index 0 Param 1=&FloatFromCurve(FloatCurve=DataInterface() {propertyValues {Curve {keys [
         {arriveTangent=-9.6411, interpMode=RCIM_Cubic, leaveTangent=-9.6411, tangentMode=RCTM_User, time=0, value=1}
         <at 0.2 0>
         {arriveTangent=1.25, interpMode=RCIM_Cubic, leaveTangent=1.25, time=0.4, value=0}
         {interpMode=RCIM_Cubic, leaveTangent=16.4602, time=1, value=1}
        ]}}})
    Index 0 Param 2=&FloatFromCurve(
     FloatCurve=DataInterface() {propertyValues {Curve {keys [
         {arriveTangent=2.4843, interpMode=RCIM_Cubic, leaveTangent=2.4843, tangentMode=RCTM_User, time=0, value=0}
         {interpMode=RCIM_Cubic, leaveTangent=-0.8461, time=1, value=1}
        ]}}}
     Scale Curve=2
    )
    Index 0 Param 4=&RandomRangeFloat(Minimum=0.4, Maximum=0.6)
    Module.Param0WriteEnabled=true
    Module.Param1WriteEnabled=true
    Module.Param2WriteEnabled=false
    Module.Param3WriteEnabled=true
   )
  ]
  rendererClasses [&NiagaraMeshRendererProperties]
  renderers [{rendererClass=&NiagaraMeshRendererProperties, rendererIndex=0}]
 }}
summary {
 systemName=NS_Emote_Bass_LetEmCook_Windwhirl
 emitters [{emitterName=Windwhirl, rendererClasses [&NiagaraMeshRendererProperties]}]
 userVariables [User.bIsFrontEndPreview {type=NiagaraBool}]
}
